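# Looping stereo MetaSound source for the Masumune vehicle engine, built from NeoCar motor loops.
# Smoothed RPM crossfades and pitches idle/on-load loops into throttle-filtered engine and exhaust buses, while speed-driven digital layers join the exhaust bus.
# An on-play envelope shapes startup, and BonusSpeed raises pitch and ducks motor loops.
# Shared patches apply volume, perspective EQ and azimuth panning; exhaust pop, waveshaper and transmission inputs are unconnected.
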MetaSoundSource /VehicleAudio/EngineAudio/Masumune/MSS_EngineAudio_Masumune
inputs [
 Acceleration {type=Float}
 Blowoff.Volume {type=Float, access=Constructor}
 BonusSpeed {type=Float}
 Character.Affiliation {type=Enum:FortAudioAffiliation}
 Character.Health {type=Float}
 Character.UniqueID {type=Int32}
 Downshift {type=Trigger}
 Engine.Throttle.EQ.FilterQ {type=Float, access=Constructor}
 Engine.Throttle.EQ.FilterType {type=Enum:BiquadFilterType, access=Constructor}
 Engine.Throttle.EQ.OffFreq {type=Float, access=Constructor}
 Engine.Throttle.EQ.OffGain {type=Float, access=Constructor}
 Engine.Throttle.EQ.OnFreq {type=Float, access=Constructor}
 Engine.Throttle.EQ.OnGain {type=Float, access=Constructor}
 Engine.Throttle.InterpTime {type=Time, access=Constructor}
 Engine.Throttle.Volume.Off {type=Float, access=Constructor}
 Engine.Throttle.Volume.On {type=Float, access=Constructor}
 EngineBlowoff {type=Trigger}
 Exhaust.Pops.Blowoff.Volume {type=Float, access=Constructor}
 Exhaust.Pops.Long {type=WaveAsset:Array, access=Constructor}
 Exhaust.Pops.Pitch {type=Float, access=Constructor}
 Exhaust.Pops.Shift.Volume {type=Float, access=Constructor}
 Exhaust.Pops.Short {type=WaveAsset:Array, access=Constructor}
 Exhaust.Throttle.EQ.FilterQ {type=Float, access=Constructor}
 Exhaust.Throttle.EQ.FilterType {type=Enum:BiquadFilterType, access=Constructor}
 Exhaust.Throttle.EQ.OffFreq {type=Float, access=Constructor}
 Exhaust.Throttle.EQ.OffGain {type=Float, access=Constructor}
 Exhaust.Throttle.EQ.OnFreq {type=Float, access=Constructor}
 Exhaust.Throttle.EQ.OnGain {type=Float, access=Constructor}
 Exhaust.Throttle.InterpTime {type=Time, access=Constructor}
 Exhaust.Throttle.Volume.Off {type=Float, access=Constructor}
 Exhaust.Throttle.Volume.On {type=Float, access=Constructor}
 Exhaust.Throttle.Waveshaper.Amount {type=Float, access=Constructor}
 Exhaust.Throttle.Waveshaper.AmountOffScale {type=Float, access=Constructor}
 Exhaust.Throttle.Waveshaper.AmountOnScale {type=Float, access=Constructor}
 Exhaust.Throttle.Waveshaper.Bias {type=Float, access=Constructor}
 Exhaust.Throttle.Waveshaper.Bypass {type=Bool, access=Constructor}
 Exhaust.Throttle.Waveshaper.Crossover {type=Float, access=Constructor}
 Exhaust.Throttle.Waveshaper.Mix {type=Float, access=Constructor}
 Exhaust.Throttle.Waveshaper.OutputGain {type=Float, access=Constructor}
 Exhaust.Throttle.Waveshaper.Type {type=Enum:WaveShaperType, access=Constructor}
 Fort.Vehicle.Acceleration {type=Float}
 Fort.Vehicle.BlowoffStrength {type=Float}
 Fort.Vehicle.BonusSpeed {type=Float}
 Fort.Vehicle.EnableShifting {type=Bool}
 Fort.Vehicle.Gear {type=Int32}
 Fort.Vehicle.IsGrounded {type=Bool}
 Fort.Vehicle.Load {type=Float}
 Fort.Vehicle.NormalSpeed {type=Float}
 Fort.Vehicle.OnBlowoff {type=Trigger}
 Fort.Vehicle.OnDownshift {type=Trigger}
 Fort.Vehicle.OnUpshift {type=Trigger}
 Fort.Vehicle.Pitch {type=Float}
 Fort.Vehicle.RPM {type=Float}
 Fort.Vehicle.Throttle {type=Float}
 GearChangeTime {type=Time, access=Constructor}
 Interior.EQ.CutoffFrequency {type=Float, access=Constructor}
 Interior.EQ.FilterType {type=Enum:BiquadFilterType, access=Constructor}
 Interior.StereoEffect.Crossover {type=Float, access=Constructor}
 Interior.StereoEffect.Mix {type=Float, access=Constructor}
 IsLocal {type=Bool}
 Mix.Azimuth.Engine.FilterType {type=Enum:BiquadFilterType, access=Constructor}
 Mix.Azimuth.Engine.Freq.Max {type=Float, access=Constructor}
 Mix.Azimuth.Engine.Freq.Min {type=Float, access=Constructor}
 Mix.Azimuth.Engine.MinVolume {type=Float, access=Constructor}
 Mix.Azimuth.Exhaust.FilterType {type=Enum:BiquadFilterType, access=Constructor}
 Mix.Azimuth.Exhaust.Freq.Max {type=Float, access=Constructor}
 Mix.Azimuth.Exhaust.Freq.Min {type=Float, access=Constructor}
 Mix.Azimuth.Exhaust.MinVolume {type=Float, access=Constructor}
 Mix.MixPerspective {type=Int32}
 Mix.Mute.Engine {type=Bool, access=Constructor}
 Mix.Mute.Exhaust {type=Bool, access=Constructor}
 Mix.Mute.Interior {type=Bool, access=Constructor}
 Mix.Perspective.Exterior.Engine.EQ {type=Float:Array, access=Constructor}
 Mix.Perspective.Exterior.Exhaust.EQ {type=Float:Array, access=Constructor}
 Mix.Perspective.Exterior.Volume {type=Float:Array, access=Constructor}
 Mix.Perspective.Interior.Engine.EQ {type=Float:Array, access=Constructor}
 Mix.Perspective.Interior.Exhaust.EQ {type=Float:Array, access=Constructor}
 Mix.Perspective.Interior.Volume {type=Float:Array, access=Constructor}
 Multichannel {type=REV Asset, access=Constructor}
 NormalizedSpeed {type=Float}
 Pitch {type=Float}
 Quality {type=Bool}
 REV.EnableShifting {type=Bool}
 REV.Gear {type=Int32}
 REV.RPM {type=Float}
 REV.Throttle {type=Float}
 REV.Velocity {type=Float}
 Settings {type=REV Settings, access=Constructor}
 Transmission.Enabled {type=Bool}
 Transmission.Gear.Volume {type=Float:Array, access=Constructor}
 Transmission.MaxFreq {type=Float, access=Constructor}
 Transmission.MinFreq {type=Float, access=Constructor}
 Transmission.Throttle.Volume.Off {type=Float, access=Constructor}
 Transmission.Throttle.Volume.On {type=Float, access=Constructor}
 Transmission.Volume.Overall {type=Float, access=Constructor}
 UE.Source.OnPlay {type=Trigger}
 UE.Source.Orientation.Azimuth {type=Float}
 UE.Source.Orientation.Elevation {type=Float}
 Upshift {type=Trigger}
]
interfaces [Fort.Character:1.0, Fort.Vehicle:1.0, REV:1.0, UE.OutputFormat.Stereo:1.0, UE.Source.Orientation:1.0, UE.Source:1.1]
outputs [
 "UE.OutputFormat.Stereo.Audio:0" {type=Audio}
 "UE.OutputFormat.Stereo.Audio:1" {type=Audio}
]
pages [Default {
  input_defaults {
   Acceleration=0
   Blowoff.Volume=0
   BonusSpeed=0
   Character.Affiliation=0
   Character.Health=0
   Character.UniqueID=0
   Downshift=null
   Engine.Throttle.EQ.FilterQ=1
   Engine.Throttle.EQ.FilterType=1
   Engine.Throttle.EQ.OffFreq=250
   Engine.Throttle.EQ.OffGain=0
   Engine.Throttle.EQ.OnFreq=20
   Engine.Throttle.EQ.OnGain=0
   Engine.Throttle.InterpTime=0.1
   Engine.Throttle.Volume.Off=-2
   Engine.Throttle.Volume.On=0
   EngineBlowoff=null
   Exhaust.Pops.Blowoff.Volume=0
   Exhaust.Pops.Pitch=0
   Exhaust.Pops.Shift.Volume=0
   Exhaust.Throttle.EQ.FilterQ=1
   Exhaust.Throttle.EQ.FilterType=4
   Exhaust.Throttle.EQ.OffFreq=200
   Exhaust.Throttle.EQ.OffGain=-5
   Exhaust.Throttle.EQ.OnFreq=200
   Exhaust.Throttle.EQ.OnGain=0
   Exhaust.Throttle.InterpTime=0.1
   Exhaust.Throttle.Volume.Off=-2
   Exhaust.Throttle.Volume.On=0
   Exhaust.Throttle.Waveshaper.Amount=7
   Exhaust.Throttle.Waveshaper.AmountOffScale=0
   Exhaust.Throttle.Waveshaper.AmountOnScale=1
   Exhaust.Throttle.Waveshaper.Bias=0
   Exhaust.Throttle.Waveshaper.Bypass=null
   Exhaust.Throttle.Waveshaper.Crossover=272
   Exhaust.Throttle.Waveshaper.Mix=1
   Exhaust.Throttle.Waveshaper.OutputGain=0.5
   Exhaust.Throttle.Waveshaper.Type=4
   Fort.Vehicle.Acceleration=0
   Fort.Vehicle.BlowoffStrength=0
   Fort.Vehicle.BonusSpeed=0
   Fort.Vehicle.EnableShifting=null
   Fort.Vehicle.Gear=0
   Fort.Vehicle.IsGrounded="<MetaSoundBuilderResult.SUCCEEDED: 0>"
   Fort.Vehicle.Load=0
   Fort.Vehicle.NormalSpeed=0
   Fort.Vehicle.OnBlowoff=null
   Fort.Vehicle.OnDownshift=null
   Fort.Vehicle.OnUpshift=null
   Fort.Vehicle.Pitch=0
   Fort.Vehicle.RPM=0
   Fort.Vehicle.Throttle=1
   GearChangeTime=0.1
   Interior.EQ.CutoffFrequency=5500
   Interior.EQ.FilterType=0
   Interior.StereoEffect.Crossover=494
   Interior.StereoEffect.Mix=0.2
   IsLocal=null
   Mix.Azimuth.Engine.FilterType=0
   Mix.Azimuth.Engine.Freq.Max=20000
   Mix.Azimuth.Engine.Freq.Min=5500
   Mix.Azimuth.Engine.MinVolume=-3
   Mix.Azimuth.Exhaust.FilterType=1
   Mix.Azimuth.Exhaust.Freq.Max=20
   Mix.Azimuth.Exhaust.Freq.Min=250
   Mix.Azimuth.Exhaust.MinVolume=-3
   Mix.MixPerspective=0
   Mix.Mute.Engine=null
   Mix.Mute.Exhaust=null
   Mix.Mute.Interior=null
   Multichannel=null
   NormalizedSpeed=0
   Pitch=0
   Quality=null
   REV.EnableShifting=null
   REV.Gear=6
   REV.RPM=0
   REV.Throttle=0
   REV.Velocity=0
   Settings=&/REVSetting_Masumune
   Transmission.Enabled=null
   Transmission.MaxFreq=250
   Transmission.MinFreq=60
   Transmission.Throttle.Volume.Off=0
   Transmission.Throttle.Volume.On=0
   Transmission.Volume.Overall=0
   UE.Source.OnPlay=null
   UE.Source.Orientation.Azimuth=-7.1997
   UE.Source.Orientation.Elevation=0
   Upshift=null
   Exhaust.Pops.Long []
   Exhaust.Pops.Short []
   Mix.Perspective.Exterior.Engine.EQ [90,1,0,1]
   Mix.Perspective.Exterior.Exhaust.EQ [120,1,0,1]
   Mix.Perspective.Exterior.Volume [0,0,-96]
   Mix.Perspective.Interior.Engine.EQ [20,1,0,1]
   Mix.Perspective.Interior.Exhaust.EQ [20,1,0,1]
   Mix.Perspective.Interior.Volume [-96,-96,0]
   Transmission.Gear.Volume [0,0,0,0,0,0,0]
  }
  nodes [
   DecibelstoLinearGain_2 = UE.Decibels to Linear Gain:Float() {also {defaults {Decibels=-4}}}
   ADEnvelopeADEnvelope_0 = AD Envelope.AD Envelope:Float(Trigger=^UE.Source.OnPlay) {also {defaults {Decay Curve=2, Decay Time=0.5}}}
   n1 = MapRange.MapRange:Float(In=@ADEnvelopeADEnvelope_0.Out Envelope) {also {defaults {Out Range B=0}}}
   n2 = UE.Add:Float(AdditionalOperands=Max.Max:Float(B=@n1.Out Value, A=Max.Max:Float(B=^Fort.Vehicle.RPM, A=^REV.RPM).Value).Value)
   n3 = VariableAccessor.Float(Variable=VariableMutator.Float(Value=Clamp.Clamp:Float(In=@n2.Out) {also {defaults {Max=1}}}.Value, Variable=InitVariable.Float().Variable).Variable)
   n4 = &InterpSpeed(FloatInput=@n3.Value) {also {defaults {FallingInterpSpeed=0.01, RisingInterpSpeed=0.5}}}
   n5 = Fortnite.CrossfadeSegmentManager:Float(In=@n4.FloatOutput) {also {defaults {Max=1.01, MaxRange=0, Min=0.026, MinRange=0.633}}}
   DecibelstoLinearGain_1 = UE.Decibels to Linear Gain:Float() {also {defaults {Decibels=-4}}}
   n6 = Fortnite.CrossfadeSegmentManager:Float(In=@n4.FloatOutput) {also {defaults {Max=0.659, MaxRange=0.633, Min=-0.1, MinRange=0}}}
   n7 = VariableDeferredAccessor.Float(Variable=InitVariable.Float().Variable)
   n8 = UE.Add:Float(
    PrimaryOperand=MapRange.MapRange:Float(In=@n4.FloatOutput) {also {defaults {Out Range A=-3, Out Range B=17.6}}}.Out Value
    AdditionalOperands=@n7.Value
   )
   WavePlayer_2 = UE.Wave Player:Mono(Pitch Shift=@n8.Out, Play=@n6.Activate, Stop=@n6.Deactivate) {also {
     defaults {Loop="<MetaSoundBuilderResult.SUCCEEDED: 0>", Wave Asset=&/NeoCar_Motor_Idle_Engine}
     version [1,1]
    }}
   WavePlayer_4 = UE.Wave Player:Mono(Pitch Shift=@n8.Out, Play=@n5.Activate, Stop=@n5.Deactivate) {also {
     defaults {Loop="<MetaSoundBuilderResult.SUCCEEDED: 0>", Wave Asset=&/NeoCar_Motor_OnLoad_Exhaust}
     version [1,1]
    }}
   AudioMixerAudioMixerMono2_0 = "AudioMixer.Audio Mixer (Mono, 2)"(
    Gain 1=UE.Multiply:Float(AdditionalOperands=@DecibelstoLinearGain_2.Linear Gain, PrimaryOperand=@n5.Volume).Out
    Gain 0=UE.Multiply:Float(AdditionalOperands=@DecibelstoLinearGain_1.Linear Gain, PrimaryOperand=@n6.Volume).Out
    In 0=@WavePlayer_2.Out Mono
    In 1=@WavePlayer_4.Out Mono
   )
   n9 = UE.Add:Float(AdditionalOperands=^Fort.Vehicle.Throttle, PrimaryOperand=^REV.Throttle)
   n10 = VariableDeferredAccessor.Float(Variable=InitVariable.Float().Variable)
   n11 = VariableMutator.Float(
    Value=Clamp.Clamp:Float(In=UE.Add:Float(PrimaryOperand=@n9.Out, AdditionalOperands=@n1.Out Value).Out) {also {defaults {Max=1, Min=0}}}.Value
    Variable=@n10.Variable
   )
   n12 = VariableAccessor.Float(Variable=VariableMutator.Float(Variable=InitVariable.Float().Variable, Value=Max.Max:Float(A=^BonusSpeed, B=^Fort.Vehicle.BonusSpeed).Value).Variable)
   n13 = /FortMetasound/Patches/MapRange_DecibelOutput(In=&InterpSpeed(FloatInput=VariableAccessor.Float(Variable=@n12.Variable).Value) {also {defaults {FallingInterpSpeed=0.5, RisingInterpSpeed=0.1}}}.FloatOutput) {also {defaults {Out Range B=-6}}}
   AudioMixerAudioMixerMono3_0 = "AudioMixer.Audio Mixer (Mono, 3)"(
    In 1=/VehicleAudio/Patches/DM_EngineAudio_ThrottleVolumeFilter(
     Filter.FilterQ=^Engine.Throttle.EQ.FilterQ
     Filter.FilterType=^Engine.Throttle.EQ.FilterType
     Filter.OffThrottleFreq=^Engine.Throttle.EQ.OffFreq
     Filter.OnThrottleGain=^Engine.Throttle.EQ.OffGain
     Filter.OnThrottleFreq=^Engine.Throttle.EQ.OnFreq
     Filter.OffThrottleGain=^Engine.Throttle.EQ.OnGain
     Interp Time=^Engine.Throttle.InterpTime
     Volume.OffThrottle=^Engine.Throttle.Volume.Off
     Volume.OnThrottle=^Engine.Throttle.Volume.On
     AudioIn=@AudioMixerAudioMixerMono2_0.Out
     REV.Throttle=VariableAccessor.Float(Variable=@n11.Variable).Value
    ).AudioOut
    Gain 1=@n13.Out Value
   )
   n14 = Fortnite.CrossfadeSegmentManager:Float(In=@n4.FloatOutput) {also {defaults {Max=0.659, MaxRange=0.633, Min=-0.1, MinRange=0}}}
   n15 = Fortnite.CrossfadeSegmentManager:Float(In=@n4.FloatOutput) {also {defaults {Max=1.01, MaxRange=0, Min=0.026, MinRange=0.633}}}
   WavePlayer_1 = UE.Wave Player:Mono(Pitch Shift=@n8.Out, Play=@n15.Activate, Stop=@n15.Deactivate) {also {
     defaults {Loop="<MetaSoundBuilderResult.SUCCEEDED: 0>", Wave Asset=&/NeoCar_Motor_OnLoad_Engine}
     version [1,1]
    }}
   WavePlayer_3 = UE.Wave Player:Mono(Pitch Shift=@n8.Out, Play=@n14.Activate, Stop=@n14.Deactivate) {also {
     defaults {Loop="<MetaSoundBuilderResult.SUCCEEDED: 0>", Wave Asset=&/NeoCar_Motor_Idle_Exhaust}
     version [1,1]
    }}
   AudioMixerAudioMixerMono2_1 = "AudioMixer.Audio Mixer (Mono, 2)"(
    Gain 0=UE.Multiply:Float(PrimaryOperand=@n14.Volume) {also {defaults {AdditionalOperands=0.7}}}.Out
    Gain 1=UE.Multiply:Float(PrimaryOperand=@n15.Volume) {also {defaults {AdditionalOperands=0.7}}}.Out
    In 1=@WavePlayer_1.Out Mono
    In 0=@WavePlayer_3.Out Mono
   )
   AudioMixerAudioMixerMono2_2 = "AudioMixer.Audio Mixer (Mono, 2)"(In 1=&DM_EngineAudio_ThrottleVolumeFilter(
     Filter.FilterQ=^Exhaust.Throttle.EQ.FilterQ
     Filter.FilterType=^Exhaust.Throttle.EQ.FilterType
     Filter.OffThrottleFreq=^Exhaust.Throttle.EQ.OffFreq
     Filter.OffThrottleGain=^Exhaust.Throttle.EQ.OffGain
     Filter.OnThrottleFreq=^Exhaust.Throttle.EQ.OnFreq
     Filter.OnThrottleGain=^Exhaust.Throttle.EQ.OnGain
     Interp Time=^Exhaust.Throttle.InterpTime
     Volume.OffThrottle=^Exhaust.Throttle.Volume.Off
     Volume.OnThrottle=^Exhaust.Throttle.Volume.On
     AudioIn=@AudioMixerAudioMixerMono2_1.Out
     REV.Throttle=@n10.Value
    ).AudioOut)
   DecibelstoLinearGain_4 = UE.Decibels to Linear Gain:Float() {also {defaults {Decibels=-10}}}
   n16 = UE.InterpTo:Audio(Target=^Fort.Vehicle.BonusSpeed) {also {defaults {Interp Time=1}}}
   n17 = Fortnite.CrossfadeSegmentManager:Float(In=@n16.Value) {also {defaults {Max=1.1, MaxCurve=2, MaxRange=0, MinRange=0.5}}}
   DecibelstoLinearGain_3 = UE.Decibels to Linear Gain:Float() {also {defaults {Decibels=-8}}}
   n18 = Max.Max:Float(B=^Fort.Vehicle.NormalSpeed, A=^NormalizedSpeed)
   n19 = VariableAccessor.Float(Variable=VariableMutator.Float(Value=UE.Add:Float(PrimaryOperand=@n18.Value).Out, Variable=InitVariable.Float().Variable).Variable)
   n20 = UE.Add:Float(
    PrimaryOperand=MapRange.MapRange:Float(In=@n19.Value) {also {defaults {Out Range A=-2, Out Range B=17.5}}}.Out Value
    AdditionalOperands=MapRange.MapRange:Float(In=^Fort.Vehicle.BonusSpeed) {also {defaults {Out Range B=3}}}.Out Value
   )
   n21 = UE.Add:Float(
    AdditionalOperands=^Fort.Vehicle.Pitch
    PrimaryOperand=UE.InterpTo:Audio(Target=MapRange.MapRange:Float(In=@n12.Value) {also {defaults {Out Range B=0.75}}}.Out Value) {also {defaults {Interp Time=0.15}}}.Value
   )
   n22 = UE.Multiply:Float(PrimaryOperand=@ADEnvelopeADEnvelope_0.Out Envelope) {also {defaults {AdditionalOperands=-12}}}
   n23 = VariableAccessor.Float(Variable=VariableMutator.Float(Value=UE.Add:Float(AdditionalOperands=@n21.Out, PrimaryOperand=@n22.Out).Out, Variable=@n7.Variable).Variable)
   WavePlayer_0 = UE.Wave Player:Mono(
    Play=^UE.Source.OnPlay
    Pitch Shift=&InterpSpeed(FloatInput=UE.Add:Float(AdditionalOperands=@n20.Out, PrimaryOperand=@n23.Value).Out) {also {defaults {RisingInterpSpeed=1.2}}}.FloatOutput
   ) {also {
     defaults {Loop="<MetaSoundBuilderResult.SUCCEEDED: 0>", Wave Asset=&/NeoCar_Motor_OffLoad_Low_Digital}
     version [1,1]
    }}
   WavePlayer_5 = UE.Wave Player:Mono(Play=@n17.Activate, Stop=@n17.Deactivate, Pitch Shift=MapRange.MapRange:Float(In=@n16.Value) {also {defaults {Out Range B=28}}}.Out Value) {also {
     defaults {Loop="<MetaSoundBuilderResult.SUCCEEDED: 0>", Wave Asset=&/NeoCar_Motor_OnLoad_Hi_Digital}
     version [1,1]
    }}
   AudioMixerAudioMixerMono2_3 = "AudioMixer.Audio Mixer (Mono, 2)"(
    Gain 1=UE.Multiply:Float(AdditionalOperands=@DecibelstoLinearGain_4.Linear Gain, PrimaryOperand=@n17.Volume).Out
    Gain 0=UE.Multiply:Float(
     AdditionalOperands=@DecibelstoLinearGain_3.Linear Gain
     PrimaryOperand=&MapRange_DecibelOutput(In=^Fort.Vehicle.BonusSpeed) {also {defaults {Out Range B=3}}}.Out Value
    ).Out
    In 0=@WavePlayer_0.Out Mono
    In 1=@WavePlayer_5.Out Mono
   )
   n24 = VariableMutator.Float(Value=@ADEnvelopeADEnvelope_0.Out Envelope, Variable=InitVariable.Float().Variable)
   AudioMixerAudioMixerStereo2_0 = "AudioMixer.Audio Mixer (Stereo, 2)"(
    In 0 L=@AudioMixerAudioMixerMono2_2.Out
    In 0 R=@AudioMixerAudioMixerMono2_2.Out
    In 1 L=@AudioMixerAudioMixerMono2_3.Out
    In 1 R=@AudioMixerAudioMixerMono2_3.Out
    Gain 0=@n13.Out Value
    Gain 1=UE.Subtract:Float(AdditionalOperands=VariableAccessor.Float(Variable=@n24.Variable).Value) {also {defaults {PrimaryOperand=1}}}.Out
   )
   n25 = /VehicleAudio/Patches/DM_EngineAudio_VolumeMixer(
    MixP=^Mix.MixPerspective
    MixP_Exterior=^Mix.Perspective.Exterior.Volume
    MixP_Interior=^Mix.Perspective.Interior.Volume
    UE.Source.OnPlay=^UE.Source.OnPlay
    EngineL=@AudioMixerAudioMixerMono3_0.Out
    EngineR=@AudioMixerAudioMixerMono3_0.Out
    ExhaustL=@AudioMixerAudioMixerStereo2_0.Out L
    ExhaustR=@AudioMixerAudioMixerStereo2_0.Out R
    Acceleration=Max.Max:Float(A=^Acceleration, B=^Fort.Vehicle.Acceleration).Value
    BonusSpeed=Max.Max:Float(A=^BonusSpeed, B=^Fort.Vehicle.BonusSpeed).Value
   ) {also {defaults {Volume_Engine=0, Volume_Exhaust=0, Volume_Interior=0, Volume_REV=0}}}
   n26 = /VehicleAudio/Patches/DM_EngineAudio_MixPFilters(
    MixP=^Mix.MixPerspective
    MixP.Exterior.Engine.EQ=^Mix.Perspective.Exterior.Engine.EQ
    MixP.Exterior.Exhaust.EQ=^Mix.Perspective.Exterior.Exhaust.EQ
    MixP.Interior.Engine.EQ=^Mix.Perspective.Interior.Engine.EQ
    MixP.Interior.Exhaust.EQ=^Mix.Perspective.Interior.Exhaust.EQ
    UE.Source.OnPlay=^UE.Source.OnPlay
    EngineL=@n25.EngineL
    EngineR=@n25.EngineR
    ExhaustL=@n25.ExhaustL
    ExhaustR=@n25.ExhaustR
    InteriorL=@n25.InteriorL
    InteriorR=@n25.InteriorR
   )
   DM_EngineAudio_AzimuthPanner_0 = /VehicleAudio/Patches/DM_EngineAudio_AzimuthPanner(
    EnginefilterType=^Mix.Azimuth.Engine.FilterType
    Engine_MaxFreq=^Mix.Azimuth.Engine.Freq.Max
    Engine_MinFreq=^Mix.Azimuth.Engine.Freq.Min
    Engine_MinVolume=^Mix.Azimuth.Engine.MinVolume
    ExhaustFilterType=^Mix.Azimuth.Exhaust.FilterType
    Exhaust_MaxFreq=^Mix.Azimuth.Exhaust.Freq.Max
    Exhaust_MinFreq=^Mix.Azimuth.Exhaust.Freq.Min
    Exhaust_MinVolume=^Mix.Azimuth.Exhaust.MinVolume
    EngineMute=^Mix.Mute.Engine
    ExhaustMute=^Mix.Mute.Exhaust
    InteriorMute=^Mix.Mute.Interior
    UE.Source.Orientation.Azimuth=^UE.Source.Orientation.Azimuth
    EngineL=@n26.EngineL
    EngineR=@n26.EngineR
    ExhaustL=@n26.ExhaustL
    ExhaustR=@n26.ExhaustR
    InteriorL=@n26.InteriorL
    InteriorR=@n26.InteriorR
   )
   DecibelstoLinearGain_0 = UE.Decibels to Linear Gain:Float() {also {defaults {Decibels=-3}}}
   /FortMetasound/Patches/PrintFloat()
   TriggerOnValueChange_0 = UE.Trigger On Value Change:Float(Value=^BonusSpeed)
   PrintLogPrintLog_0 = Print Log.Print Log:Bool(
    Value To Log=/FortMetasound/Patches/CompareFloat(A=^BonusSpeed) {also {defaults {B=1, Type=0}}}.Result
    Trigger=@TriggerOnValueChange_0.Trigger
   ) {also {defaults {Label=bool}}}
   VariableAccessor.Float(Variable=@n3.Variable)
   VariableAccessor.Float(Variable=@n19.Variable)
   VariableMutator.Int32(Value=UE.Add:Int32(PrimaryOperand=Max.Max:Int32(B=^Fort.Vehicle.Gear, A=^REV.Gear).Value).Out, Variable=InitVariable.Int32().Variable)
   TriggerAnyTriggerAny3_0 = "TriggerAny.Trigger Any (3)"(In 0=^Downshift, In 1=^Upshift)
   VariableMutator.Trigger(Variable=InitVariable.Trigger().Variable, Value=@TriggerAnyTriggerAny3_0.Out)
  ]
  links [
   {dst="@UE.OutputFormat.Stereo.Audio:0", src=DM_EngineAudio_AzimuthPanner_0.Out L}
   {dst="@UE.OutputFormat.Stereo.Audio:1", src=DM_EngineAudio_AzimuthPanner_0.Out R}
  ]
 }]
properties {
 assetImportData=&self:AssetImportData
 attenuationSettings=&Vehicle_Engine_Att
 audioPropertiesSheet=&APS_Cosmetic_Vehicle_Engine
 bLooping=true
 duration=10000
 maxDistance=12000
 numChannels=2
 outputFormat=Stereo
 priority=80
 sampleRate=48000
 soundClassObject=&Fort_Vehicles_Engine
 volume=0.5
 concurrencySet [&Conc_Veh_EngineAudio_LowSpec, &Conc_Veh_EngineAudio_LowSpec_Actor]
 modulationSettings {
  highpassRouting=Union
  lowpassRouting=Union
  pitchRouting=Union
  volumeRouting=Union
  highpassModulationDestination {bEnableModulation=true, modulators [&PP_FortVehicles_Engine_HPF]}
  lowpassModulationDestination {bEnableModulation=true, modulators [&Patch_FortVehicles_Engine_LPF]}
  pitchModulationDestination {bEnableModulation=true, modulators [&DM_Bus_Vehicle_Engine_Pitch]}
  volumeModulationDestination {
   bEnableModulation=true
   modulators [&DM_VehicleEnginePatch, &Patch_FortVehicles_Engine_Vol]
  }
 }
}
